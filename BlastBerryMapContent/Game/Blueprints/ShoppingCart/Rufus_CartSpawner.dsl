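; Shopping cart spawner Blueprint that can override the max health of the vehicle it spawns.
; The spawn vehicle event runs the parent logic, then on authority checks whether the shouldOverrideVehicles scalable float is a valid curve setting and true.
; If so, it applies the GE_HealthBuff gameplay effect to the spawned vehicle as a global environment effect, provided both references are valid.
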(event EventSpawnVehicle
  (|CallParentFunction)
  ;[0 0 0] Apply a Vehicle Gamplay Effect that will Override the Max Health of the Vehicle after it is Spawned
  (Networking|SwitchHasAuthority
    (:Authority
      (bind _bshouldoverride (Default|ShouldUseVehicleOverrideGE))
      (if _bshouldoverride
        (bind _spawnedvehicle (Variables|Default|GetSpawnedVehicle)
          (:then
            (bind _ge_healthbuff (Variables|Default|GetGE_HealthBuff)
              (:then
                (Environment|ApplyGlobalEnvironmentGameplayEffecttoActor _spawnedvehicle _ge_healthbuff))
              (:else)))
          (:else))))))

(fn ShouldUseVehicleOverrideGE
  (return (and (ScalableFloat|IsSettingCurveValid (Variables|Default|GetshouldOverrideVehicles)) (ScalableFloat|GetValueasBool (Variables|Default|GetshouldOverrideVehicles)))))

(fn ConstructionScript)
