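; Drivable shopping cart vehicle Blueprint adding water effects, flip and impulse ejection, and collision responses.
; Entering water starts attached particles, audio and a looping timer tracking surface height and speed, offset per Rufus lake type.
; A timer started on pawn entry ejects players on excessive roll, as do hotfix-gated impulses on authority.
; Dedicated-server hits damage weak props at speed, or bump via a silent-sound ability impact broadcast when the cart is empty.
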(event Collision|EventActorBeginOverlap
  (params (OtherActor :type Actor%20Object%20Reference))
  ;[1 1 1] Water VFX offset
  (bind _asbp_rufus_water (Utilities|Casting|CastToBP_Rufus_Water OtherActor))
  (CallFunction|UpdateWaterVFXOffset true _asbp_rufus_water))

(event Collision|EventActorEndOverlap
  (params (OtherActor :type Actor%20Object%20Reference))
  ;[1 1 1] Water VFX offset
  (bind _asbp_rufus_water (Utilities|Casting|CastToBP_Rufus_Water OtherActor))
  (CallFunction|UpdateWaterVFXOffset false _asbp_rufus_water))

(event Vehicle|Water|EventOnEnteredWaterVolume
  (params (WaterSurfacePoint :type Vector%20%28by%20ref%29 :default "0, 0, 0"))
  (bind _returnvalue (CallFunction|GetWaterLevel))
  (bind (_returnvalue_x _returnvalue_y _returnvalue_z) (Math|Vector|BreakVector (Transformation|GetActorLocation)))
  (bind _const_water_fxoffset_z (Variables|Default|GetConst%20Water%20FXOffset%20Z))
  (bind _returnvalue_2 (+ _returnvalue _const_water_fxoffset_z))
  (bind _returnvalue_3 (- _returnvalue_2 _returnvalue_z))
  (bind _skeletalmesh (Variables|Vehicle|GetSkeletalMesh))
  ;[1 1 1] FX for water ineraction
  (Networking|CanExecuteCosmeticEvents
    (:True
      (if (> (Rendering|GetRenderingDetailMode) 0)
        (bind _returnvalue_4 (Utilities|Time|SetTimerbyEvent (delegate Custom|CheckWaterLoop) 0.25 true))
        (bind _inlakefx (Variables|WaterInteraction|GetInLakeFX)
          (:then
            (Components|DestroyComponent _inlakefx))
          (:else
            _))
        (bind _returnvalue_5 (Effects|SpawnEmitterAttached :EmitterTemplate "/Game/WaterAndWind/Particles/WaterInteraction/P_Vehicle_In_Water_Athena.P_Vehicle_In_Water_Athena" :AttachToComponent _skeletalmesh :AttachPointName "None" :Location (Math|Vector|MakeVector 0.0 0.0 _returnvalue_3) :Scale "1.000000,1.000000,1.000000" :LocationType "KeepRelativeOffset" :bAutoDestroy true :PoolingMethod "None" :bAutoActivate true)) ; Looping water FX
        (Variables|WaterInteraction|SetInLakeFX _returnvalue_5)
        (Effects|SpawnEmitterAttached :EmitterTemplate "/Game/WaterAndWind/Particles/WaterInteraction/P_WaterSplash_Athena_Vehicle.P_WaterSplash_Athena_Vehicle" :AttachToComponent _skeletalmesh :AttachPointName "None" :Location (Math|Vector|MakeVector 0.0 0.0 _returnvalue_3) :Scale "1.000000,1.000000,1.000000" :LocationType "KeepRelativeOffset" :bAutoDestroy true :PoolingMethod "None" :bAutoActivate true) ; Splash burst FX
        (bind _inwaterloop (Variables|Default|GetInWaterLoop)
          (:then
            (Audio|Components|Audio|Stop _inwaterloop))
          (:else
            _))
        (Audio|SpawnSoundAttached "/Game/Athena/Sounds/Vehicles/GolfCart/GolfCart_Movement_Impact_Water_Cue.GolfCart_Movement_Impact_Water_Cue" (Variables|Vehicle|GetSkeletalMesh))
        (bind _returnvalue_6 (Audio|SpawnSoundAttached "/Game/Athena/Sounds/Vehicles/GolfCart/GolfCart_Movement_InWater_Loop_Cue.GolfCart_Movement_InWater_Loop_Cue" (Variables|Vehicle|GetSkeletalMesh)))
        (Variables|Default|SetInWaterLoop _returnvalue_6)))))

(event Custom|CheckWaterLoop
  (bind _returnvalue (CallFunction|GetWaterLevel))
  (bind (_returnvalue_x _returnvalue_y _returnvalue_z) (Math|Vector|BreakVector (Transformation|GetActorLocation)))
  (bind _const_water_fxoffset_z (Variables|Default|GetConst%20Water%20FXOffset%20Z))
  (bind _returnvalue_2 (+ _returnvalue _const_water_fxoffset_z))
  (bind _returnvalue_3 (- _returnvalue_2 _returnvalue_z))
  ;[1 1 1] FX for water ineraction
  (bind _inlakefx (Variables|WaterInteraction|GetInLakeFX)
    (:then
      (Transformation|SetRelativeLocation _inlakefx (Math|Vector|MakeVector 0.0 0.0 _returnvalue_3) false false)
      (Development|PrintString (Utilities|String|Append "Water Height:  " (Utilities|String|ToString(Float) _returnvalue_3)))
      (Effects|Components|ParticleSystem|SetVectorParameter (Variables|WaterInteraction|GetInLakeFX) "Offset" (Math|Vector|Normalize (Math|Transform|InverseTransformDirection (Transformation|GetWorldTransform (Variables|WaterInteraction|GetInLakeFX)) (Transformation|GetVelocity))))
      (bind _inwaterloop (Variables|Default|GetInWaterLoop)
        (:then
          (Audio|Components|Audio|SetVolumeMultiplier _inwaterloop (Math|Float|MapRangeClamped (Vehicle|GetForwardSpeedKmh) 0.0 2e+01 0.0 1.0) :!class /Script/Engine.AudioComponent))
        (:else)))
    (:else)))

(event Vehicle|Water|EventOnExitedWaterVolume
  (params (WaterSurfacePoint :type Vector%20%28by%20ref%29 :default "0, 0, 0"))
  ;[1 1 1] FX for water ineraction
  (Networking|CanExecuteCosmeticEvents
    (:True
      (Utilities|Time|ClearandInvalidateTimerbyHandle (stale Vehicle|Water|EventOnEnteredWaterVolume _returnvalue_4))
      (bind _inlakefx (Variables|WaterInteraction|GetInLakeFX)
        (:then
          (Components|Activation|SetActive _inlakefx false false)
          (bind _inwaterloop (Variables|Default|GetInWaterLoop)
            (:then
              (Audio|Components|Audio|FadeOut _inwaterloop 3.0 0.0 "Linear" :!class /Script/Engine.AudioComponent))
            (:else)))
        (:else)))))

(event EventOnPawnExitVehicle
  (params (PlayerPawn :type Fort%20Player%20Pawn%20Object%20Reference) (ExitSocketName :type Name :default "None"))
  ;[1 1 1] Adds on Eject on vehicle flip timer on pawn enter and exit vehicle.
  (|CallParentFunction PlayerPawn ExitSocketName)
  (Utilities|Time|ClearandInvalidateTimerbyHandle (Variables|Default|GetFlip%20Check%20Timer%20Handle)))

(event Custom|Check%20for%20Vehicle%20Flip
  (bind _relativerotation_roll (Math|Rotator|BreakRotator (Variables|/Script/Engine.SceneComponent|GetRelativeRotation_Roll (Variables|Vehicle|GetSkeletalMesh))))
  ;[0 1 0.08696] Check Vehicle Flip Rotation to Eject Players
  (if (>= (Math|Float|Absolute(Float) _relativerotation_roll) (Variables|Default|GetFlip%20Minimum%20Rotation))
    (Vehicle|EjectAllPlayers)))

(event EventOnPawnEnterVehicle
  (params (PlayerPawn :type Fort%20Player%20Pawn%20Object%20Reference) (SeatIdx :type Integer))
  ;[1 1 1] Adds on Eject on vehicle flip timer on pawn enter and exit vehicle.
  (|CallParentFunction PlayerPawn SeatIdx)
  (bind _returnvalue (Utilities|Time|SetTimerbyEvent (EventDispatchers|CreateEvent :!event Custom|Check%20for%20Vehicle%20Flip) 0.05 true))
  (Variables|Default|SetFlip%20Check%20Timer%20Handle _returnvalue))

(event OnComponentHit(SkeletalMesh)
  (params (HitComponent :type Primitive%20Component%20Object%20Reference) (OtherActor :type Actor%20Object%20Reference) (OtherComp :type Primitive%20Component%20Object%20Reference) (NormalImpulse :type Vector :default "0, 0, 0") (Hit_bBlockingHit :type Boolean) (Hit_bInitialOverlap :type Boolean) (Hit_Time :type Float%20%28single-precision%29) (Hit_Distance :type Float%20%28single-precision%29) (Hit_Location :type Vector :default "0, 0, 0") (Hit_ImpactPoint :type Vector :default "0, 0, 0") (Hit_Normal :type Vector :default "0, 0, 0") (Hit_ImpactNormal :type Vector :default "0, 0, 0") (Hit_PhysMat :type Physical%20Material%20Object%20Reference) (Hit_HitActor :type Actor%20Object%20Reference) (Hit_HitComponent :type Primitive%20Component%20Object%20Reference) (Hit_HitBoneName :type Name :default "None") (Hit_BoneName :type Name :default "None") (Hit_HitItem :type Integer) (Hit_ElementIndex :type Integer) (Hit_FaceIndex :type Integer) (Hit_TraceStart :type Vector :default "0, 0, 0") (Hit_TraceEnd :type Vector :default "0, 0, 0"))
  ;[1 1 1] EVENT -> On Hit Skel Mesh: Destroy Props
  (if (> (Variables|Default|GetBumpForce) 0)
    (Utilities|IsValid OtherActor
      (:Is%20Valid
        (bind _output_get (Variables|Default|SetVehicleHitActor OtherActor))
        (if (Networking|IsDedicatedServer)
          (Utilities|FlowControl|Sequence
            (:then_0
              (if (Default|ShouldDestroyProp? Hit_Normal)
                (bind _asbuilding_prop (Utilities|Casting|CastToBuildingProp (Variables|Default|GetVehicleHitActor)))
                (if (not (Actor|ActorHasTag _asbuilding_prop "NoVehicleDamage"))
                  (Default|DestroyBuildingPiece _asbuilding_prop (Variables|Default|GetMaxHealthToDestroy)))))
            (:then_1
              (bind _asfort_player_pawn (Utilities|Casting|CastToFortPlayerPawn _output_get))
              (if (Default|ShouldBump Hit_Normal _asfort_player_pawn)
                (Audio|BroadcastAbilityImpactatLocation "/Game/Sounds/Fort_Cinematics/10_seconds_silence_Cinematic_Cue.10_seconds_silence_Cinematic_Cue" self (* Hit_Normal (Variables|Default|GetBumpForce)) 771.0)))))))))

(event EventOnHandleImpulsed
  (params (ImpulsePoint :type Vector%20%28by%20ref%29 :default "0, 0, 0") (ImpulseValue :type Vector%20%28by%20ref%29 :default "0, 0, 0") (InTags :type Gameplay%20Tag%20Container%20Structure%20%28by%20ref%29) (EffectContext :type Gameplay%20Effect%20Context%20Handle%20Structure%20%28by%20ref%29) (EventInstigator :type Controller%20Object%20Reference) (ImpulseCauser :type Actor%20Object%20Reference))
  ;[0.2824 0.2235 0.1098] When hit with a large normal explosion eject all players.
  (Networking|SwitchHasAuthority
    (:Authority
      (if (and (>= (Math|Vector|VectorLengthSquared ImpulseValue) (Math|Float|Square (ScalableFloat|GetValueatLevel (Variables|Default|GetHF_EjectOnImpulseThreshold)))) (ScalableFloat|GetValueasBool (Variables|Default|GetHF_EjectOnImpulse)))
        (Vehicle|EjectAllPlayers)))))

(event EventOnAbilityImpactAtLocation
  (params (WorldContextObject :type Object%20Reference) (InSound :type Sound%20Base%20Object%20Reference) (InInstigator :type Object%20Reference) (Velocity :type Vector%20%28by%20ref%29 :default "0, 0, 0") (Radius :type Float%20%28single-precision%29) (Tag :type Name :default "None") (bSubtractLocalNoise :type Boolean))
  ;[0.2824 0.2235 0.1098] When impacted by an impulsing ability, eject all players.
  (Networking|SwitchHasAuthority
    (:Authority
      (if (ScalableFloat|GetValueasBool (Variables|Default|GetHF_EjectOnImpulse))
        (Vehicle|EjectAllPlayers)))))

(fn DestroyBuildingPiece
  (params (BuildingActor :type Building%20Actor%20Object%20Reference) (DestroyThreshold :type Float%20%28double-precision%29))
  (if (Variables|/Script/Engine.Actor|GetbCanBeDamaged BuildingActor)
    (if (<= (Health|GetMaxHealth BuildingActor :!class /Script/FortniteGame.BuildingActor) DestroyThreshold)
      (GameplayEffects|ApplyGameplayEffectToTarget (Ability|GetAbilitySystemComponent (Variables|Default|GetDriverPawn)) "/Game/Athena/DrivableVehicles/Hoverboard/GE_VehicleDamage150.GE_VehicleDamage150_C" (Ability|GetAbilitySystemComponent (Variables|Default|GetVehicleHitActor)) 0.0 (GameplayEffects|MakeEffectContext (Ability|GetAbilitySystemComponent (Variables|Default|GetDriverPawn)))))))

(fn GetBumpForceHotfix
  ;[0 0 0] Update Bump Force with Hotfix
  (bind _outxy (DataTable|EvaluateCurveTableRow "/BlastBerryMapContent/Datatables/OverrideGameData.OverrideGameData" (Variables|Default|GetBumpForceHotfix))
    (:RowFound
      (Variables|Default|SetBumpForce _outxy))))

(fn ShouldBump
  (params (Hit%20Normal :type Vector) (PlayerPawn :type Fort%20Player%20Pawn%20Object%20Reference))
  (return (Math|Boolean|ANDBoolean (< (Math|Float|Absolute(Float) (Math|Vector|DotProduct Hit%20Normal (Transformation|GetActorUpVector))) (Variables|Default|GetBumpForceAngle)) (not (Vehicle|ContainsPlayers)) (not (Vehicle|IsinVehicle PlayerPawn)))))

(fn ShouldDestroyProp?
  (params (HitNormal :type Vector))
  (bind _vehiclehitactor (Variables|Default|GetVehicleHitActor)
    (:then
      (return (and (Variables|/Script/Engine.Actor|GetbCanBeDamaged _vehiclehitactor) (and (<= (Math|Vector|DotProduct (Math|Vector|Normalize (Vehicle|GetPreviousLinearVelocity)) HitNormal) (Variables|Default|GetDestructionAngle)) (>= (Vehicle|GetPreviousForwardSpeedKmh) (Variables|Default|GetMaxSpeedToDestroy))))))
    (:else
      (return false))))

(fn UpdateWaterVFXOffset
  (params (Entering :type Boolean) (Lake%20Reference :type BP%20Rufus%20Water%20Object%20Reference))
  (if Entering
    (Utilities|Casting|CastToBP_Rufus_Water_LeakyLake Lake%20Reference
      (:then
        (Development|PrintString "Vehicle is entering Leaky Lake.")
        (Variables|Default|SetConst%20Water%20FXOffset%20Z -56.0))
      (:CastFailed
        (Utilities|Casting|CastToBP_Rufus_Water_LootLake Lake%20Reference
          (:then
            (Development|PrintString "Vehicle is entering Loot Lake.")
            (Variables|Default|SetConst%20Water%20FXOffset%20Z 0.0))
          (:CastFailed
            (Development|PrintString "Vehicle is entering a Rufus Water Body.")
            (Variables|Default|SetConst%20Water%20FXOffset%20Z -56.0)))))
    (else
      (Development|PrintString "Vehicle is exiting water."))))

(fn ConstructionScript
  (|CallParentFunction))
